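# Physics asset for the Pencil male medium-body jacket cosmetic, defining collision and joint limits for its dynamic jacket bones.
# Jacket panels at back, outer back, sides and front are two-capsule chains with limited swing and twist, hung from small MSTR root bodies.
# Those roots, the collar bodies and a necklace root join spine_03 with free angular motion.
# Larger pelvis, spine and thigh capsules represent the wearer's body.
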PhysicsAsset /BRCosmetics/Characters/Player/Male/Medium/Bodies/M_MED_Pencil_A/Jacket/Meshes/Pencil_Jacket_MedA_Ubdy_Jacket_Physics
bodies {
 collar_b_mstr [{shapeType=Capsule, length=0.505, radius=0.505, shapeName=collar_b_mstr_capsule, rotation=<rot 0 -90 0>}]
 collar_bl_mstr [{shapeType=Capsule, length=0.505, radius=0.505, shapeName=collar_bl_mstr_capsule, rotation=<rot 0 -90 0>}]
 collar_br_mstr [{shapeType=Capsule, length=0.505, radius=0.505, shapeName=collar_br_mstr_capsule, rotation=<rot 0 -90 0>}]
 collar_fl_mstr [{shapeType=Capsule, length=0.505, radius=0.505, shapeName=collar_fl_mstr_capsule, rotation=<rot 0 -90 0>}]
 collar_fr_mstr [{shapeType=Capsule, length=0.505, radius=0.505, shapeName=collar_fr_mstr_capsule, rotation=<rot 0 -90 0>}]
 collar_l_mstr [{shapeType=Capsule, length=0.505, radius=0.505, shapeName=collar_l_mstr_capsule, rotation=<rot 0 -90 0>}]
 collar_r_mstr [{shapeType=Capsule, length=0.505, radius=0.505, shapeName=collar_r_mstr_capsule, rotation=<rot 0 -90 0>}]
 dyn_jacket_bk_01 [{shapeType=Capsule, length=7.798, radius=6.0041, shapeName=dyn_puffhood_l_01_capsule, center=<4.5 0 4>, rotation=<rot -90 0 0>}]
 dyn_jacket_bk_01_MSTR [{shapeType=Capsule, length=0.505, radius=0.505, shapeName=dyn_jacket_bk_01_MSTR_capsule, rotation=<rot 0 -90 0>}]
 dyn_jacket_bk_02 [{shapeType=Capsule, length=7.798, radius=6.0041, shapeName=dyn_puffhood_l_01_capsule, center=<4.5 0 4>, rotation=<rot -90 0 0>}]
 dyn_jacket_bk_l_01 [{shapeType=Capsule, length=7.6553, radius=6.1468, shapeName=dyn_puffhood_l_01_capsule, center=<4.5 0 4>, rotation=<rot -90 0 0>}]
 dyn_jacket_bk_l_01_MSTR [{shapeType=Capsule, length=0.505, radius=0.505, shapeName=dyn_jacket_bk_l_01_MSTR_capsule, rotation=<rot 0 -90 0>}]
 dyn_jacket_bk_l_02 [{shapeType=Capsule, length=7.6553, radius=6.1468, shapeName=dyn_puffhood_l_01_capsule, center=<4.5 0 4>, rotation=<rot -90 0 0>}]
 dyn_jacket_bk_r_01 [{shapeType=Capsule, length=8.5352, radius=5.2669, shapeName=dyn_puffhood_l_01_capsule, center=<4.5 0 4>, rotation=<rot -90 0 0>}]
 dyn_jacket_bk_r_01_MSTR [{shapeType=Capsule, length=0.505, radius=0.505, shapeName=dyn_jacket_bk_r_01_MSTR_capsule, rotation=<rot 0 -90 0>}]
 dyn_jacket_bk_r_02 [{shapeType=Capsule, length=8.5352, radius=5.2669, shapeName=dyn_puffhood_l_01_capsule, center=<4.5 0 4>, rotation=<rot -90 0 0>}]
 dyn_jacket_bkout_l_01 [{shapeType=Capsule, length=9.6095, radius=5.1927, shapeName=dyn_puffhood_l_01_capsule, center=<4.5 0 4>, rotation=<rot -90 0 0>}]
 dyn_jacket_bkout_l_01_MSTR [{shapeType=Capsule, length=0.505, radius=0.505, shapeName=dyn_jacket_bkout_l_01_MSTR_capsule, rotation=<rot 0 -90 0>}]
 dyn_jacket_bkout_l_02 [{shapeType=Capsule, length=9.6095, radius=4.1927, shapeName=dyn_puffhood_l_01_capsule, center=<4.5 0 4>, rotation=<rot -90 0 0>}]
 dyn_jacket_bkout_r_01 [{shapeType=Capsule, length=8.6095, radius=5.1927, shapeName=dyn_puffhood_l_01_capsule, center=<4.5 0 4>, rotation=<rot -90 0 0>}]
 dyn_jacket_bkout_r_01_MSTR [{shapeType=Capsule, length=0.505, radius=0.505, shapeName=dyn_jacket_bkout_r_01_MSTR_capsule, rotation=<rot 0 -90 0>}]
 dyn_jacket_bkout_r_02 [{shapeType=Capsule, length=9.6095, radius=4.1927, shapeName=dyn_puffhood_l_01_capsule, center=<4.5 0 4>, rotation=<rot -90 0 0>}]
 dyn_jacket_fr_l_01 [{shapeType=Capsule, length=2.6903, radius=4.1118, shapeName=dyn_puffhood_l_01_capsule, center=<4.8666 0 4>, rotation=<rot -90 180 180>}]
 dyn_jacket_fr_l_01_MSTR [{shapeType=Capsule, length=0.505, radius=0.505, shapeName=dyn_jacket_fr_l_01_MSTR_capsule, rotation=<rot 0 -90 0>}]
 dyn_jacket_fr_l_02 [{shapeType=Capsule, length=8.5012, radius=4.1118, shapeName=dyn_puffhood_l_01_capsule, center=<6.3297 0 4>, rotation=<rot -90 180 180>}]
 dyn_jacket_fr_r_01 [{shapeType=Capsule, length=3.5, radius=4.9955, shapeName=dyn_puffhood_l_01_capsule, center=<4 0 5>, rotation=<rot 90 0 0>}]
 dyn_jacket_fr_r_01_MSTR [{shapeType=Capsule, length=0.505, radius=0.505, shapeName=dyn_jacket_fr_r_01_MSTR_capsule, rotation=<rot 0 -90 0>}]
 dyn_jacket_fr_r_02 [{shapeType=Capsule, length=7.0882, radius=4.9955, shapeName=dyn_puffhood_l_01_capsule, center=<7.1719 0 4>, rotation=<rot 90 0 0>}]
 dyn_jacket_l_01 [{shapeType=Capsule, length=7.5853, radius=6.2169, shapeName=dyn_puffhood_l_01_capsule, center=<9.319 0 4>, rotation=<rot -90 180 180>}]
 dyn_jacket_l_01_MSTR [{shapeType=Capsule, length=0.505, radius=0.505, shapeName=dyn_jacket_l_01_MSTR_capsule, rotation=<rot 0 -90 0>}]
 dyn_jacket_l_02 [{shapeType=Capsule, length=8.3353, radius=5.4669, shapeName=dyn_puffhood_l_01_capsule, center=<9.319 0 4>, rotation=<rot -90 180 180>}]
 dyn_jacket_r_01 [{shapeType=Capsule, length=8.6422, radius=5.16, shapeName=dyn_puffhood_l_01_capsule, center=<8.5601 0 4>, rotation=<rot -90 0 0>}]
 dyn_jacket_r_01_MSTR [{shapeType=Capsule, length=0.505, radius=0.505, shapeName=dyn_jacket_r_01_MSTR_capsule, rotation=<rot 0 -90 0>}]
 dyn_jacket_r_02 [{shapeType=Capsule, length=8.3922, radius=5.41, shapeName=dyn_puffhood_l_01_capsule, center=<8.5601 0 4>, rotation=<rot -90 0 0>}]
 dyn_necklace_MSTR [{shapeType=Capsule, length=0.505, radius=0.505, shapeName=dyn_necklace_MSTR_capsule, rotation=<rot 0 -90 0>}]
 pelvis [{shapeType=Capsule, length=11.0236, radius=9.9657, shapeName=spine_01_capsule, center=<-0.3055 -1.4752 0>}]
 spine_01 [{shapeType=Capsule, length=8.5456, radius=10.6999, shapeName=spine_01_capsule}]
 spine_02 [{shapeType=Capsule, length=9.0929, radius=9.2844, shapeName=spine_02_capsule, center=<0 1.1394 0>}]
 spine_03 [{shapeType=Capsule, length=6.4171, radius=9.386, shapeName=spine_03_capsule, center=<-0.1664 1.4543 0>}]
 thigh_l [{shapeType=Capsule, length=22.2046, radius=9.7954, shapeName=thigh_l_capsule, center=<-19.1504 0 0>, rotation=<rot 90 0 0>}]
 thigh_r [{shapeType=Capsule, length=21.9954, radius=10.0046, shapeName=thigh_r_capsule, center=<19.3815 0 0>, rotation=<rot 90 0 0>}]
}
constraints [
 {bone1Name=dyn_jacket_bk_01, bone2Name=dyn_jacket_bk_01_MSTR, swing1LimitDegrees=10, swing2LimitDegrees=30, twistLimitDegrees=5}
 {bone1Name=dyn_jacket_bk_02, bone2Name=dyn_jacket_bk_01, swing1LimitDegrees=10, swing2LimitDegrees=30, twistLimitDegrees=5}
 {bone1Name=dyn_jacket_bk_l_01, bone2Name=dyn_jacket_bk_l_01_MSTR, swing1LimitDegrees=10, swing2LimitDegrees=30, twistLimitDegrees=5}
 {bone1Name=dyn_jacket_bk_l_02, bone2Name=dyn_jacket_bk_l_01, swing1LimitDegrees=10, swing2LimitDegrees=30, twistLimitDegrees=5}
 {bone1Name=dyn_jacket_bk_r_01, bone2Name=dyn_jacket_bk_r_01_MSTR, swing1LimitDegrees=10, swing2LimitDegrees=30, twistLimitDegrees=5}
 {bone1Name=dyn_jacket_bk_r_02, bone2Name=dyn_jacket_bk_r_01, swing1LimitDegrees=10, swing2LimitDegrees=30, twistLimitDegrees=5}
 {bone1Name=dyn_jacket_bkout_l_01, bone2Name=dyn_jacket_bkout_l_01_MSTR, swing1LimitDegrees=10, swing2LimitDegrees=30, twistLimitDegrees=5}
 {bone1Name=dyn_jacket_bkout_l_02, bone2Name=dyn_jacket_bkout_l_01, swing1LimitDegrees=10, swing2LimitDegrees=30, twistLimitDegrees=5}
 {bone1Name=dyn_jacket_bkout_r_01, bone2Name=dyn_jacket_bkout_r_01_MSTR, swing1LimitDegrees=10, swing2LimitDegrees=30, twistLimitDegrees=5}
 {bone1Name=dyn_jacket_bkout_r_02, bone2Name=dyn_jacket_bkout_r_01, swing1LimitDegrees=10, swing2LimitDegrees=30, twistLimitDegrees=5}
 {bone1Name=dyn_jacket_fr_l_01, bone2Name=dyn_jacket_fr_l_01_MSTR, swing1LimitDegrees=60, swing2LimitDegrees=60, twistLimitDegrees=60}
 {bone1Name=dyn_jacket_fr_l_02, bone2Name=dyn_jacket_fr_l_01, swing1LimitDegrees=60, swing2LimitDegrees=60, twistLimitDegrees=60}
 {bone1Name=dyn_jacket_fr_r_01, bone2Name=dyn_jacket_fr_r_01_MSTR, swing1LimitDegrees=45, swing2LimitDegrees=25, twistLimitDegrees=50}
 {bone1Name=dyn_jacket_fr_r_02, bone2Name=dyn_jacket_fr_r_01, swing1LimitDegrees=30, swing2LimitDegrees=25, twistLimitDegrees=50}
 {bone1Name=dyn_jacket_l_01, bone2Name=dyn_jacket_l_01_MSTR, swing1LimitDegrees=45, swing2LimitDegrees=30, twistLimitDegrees=5}
 {bone1Name=dyn_jacket_l_02, bone2Name=dyn_jacket_l_01, swing1LimitDegrees=45, swing2LimitDegrees=30, twistLimitDegrees=5}
 {bone1Name=dyn_jacket_r_01, bone2Name=dyn_jacket_r_01_MSTR, swing1LimitDegrees=45, swing2LimitDegrees=20, twistLimitDegrees=5}
 {bone1Name=dyn_jacket_r_02, bone2Name=dyn_jacket_r_01, swing1LimitDegrees=45, swing2LimitDegrees=20, twistLimitDegrees=5}
 {
  bone1Name=collar_fr_mstr
  bone2Name=spine_03
  swing1LimitDegrees=45
  swing1Motion=Free
  swing2LimitDegrees=45
  swing2Motion=Free
  twistLimitDegrees=45
  twistMotion=Free
 }
 {
  bone1Name=collar_r_mstr
  bone2Name=spine_03
  swing1LimitDegrees=45
  swing1Motion=Free
  swing2LimitDegrees=45
  swing2Motion=Free
  twistLimitDegrees=45
  twistMotion=Free
 }
 {
  bone1Name=collar_br_mstr
  bone2Name=spine_03
  swing1LimitDegrees=45
  swing1Motion=Free
  swing2LimitDegrees=45
  swing2Motion=Free
  twistLimitDegrees=45
  twistMotion=Free
 }
 {
  bone1Name=collar_b_mstr
  bone2Name=spine_03
  swing1LimitDegrees=45
  swing1Motion=Free
  swing2LimitDegrees=45
  swing2Motion=Free
  twistLimitDegrees=45
  twistMotion=Free
 }
 {
  bone1Name=collar_bl_mstr
  bone2Name=spine_03
  swing1LimitDegrees=45
  swing1Motion=Free
  swing2LimitDegrees=45
  swing2Motion=Free
  twistLimitDegrees=45
  twistMotion=Free
 }
 {
  bone1Name=collar_l_mstr
  bone2Name=spine_03
  swing1LimitDegrees=45
  swing1Motion=Free
  swing2LimitDegrees=45
  swing2Motion=Free
  twistLimitDegrees=45
  twistMotion=Free
 }
 {
  bone1Name=collar_fl_mstr
  bone2Name=spine_03
  swing1LimitDegrees=45
  swing1Motion=Free
  swing2LimitDegrees=45
  swing2Motion=Free
  twistLimitDegrees=45
  twistMotion=Free
 }
 {
  bone1Name=dyn_necklace_MSTR
  bone2Name=spine_03
  swing1LimitDegrees=45
  swing1Motion=Free
  swing2LimitDegrees=45
  swing2Motion=Free
  twistLimitDegrees=45
  twistMotion=Free
 }
 {
  bone1Name=dyn_jacket_bk_01_MSTR
  bone2Name=spine_03
  swing1LimitDegrees=45
  swing1Motion=Free
  swing2LimitDegrees=45
  swing2Motion=Free
  twistLimitDegrees=45
  twistMotion=Free
 }
 {
  bone1Name=dyn_jacket_bk_r_01_MSTR
  bone2Name=spine_03
  swing1LimitDegrees=45
  swing1Motion=Free
  swing2LimitDegrees=45
  swing2Motion=Free
  twistLimitDegrees=45
  twistMotion=Free
 }
 {
  bone1Name=dyn_jacket_fr_r_01_MSTR
  bone2Name=spine_03
  swing1LimitDegrees=45
  swing1Motion=Free
  swing2LimitDegrees=45
  swing2Motion=Free
  twistLimitDegrees=45
  twistMotion=Free
 }
 {
  bone1Name=dyn_jacket_r_01_MSTR
  bone2Name=spine_03
  swing1LimitDegrees=45
  swing1Motion=Free
  swing2LimitDegrees=45
  swing2Motion=Free
  twistLimitDegrees=45
  twistMotion=Free
 }
 {
  bone1Name=dyn_jacket_bkout_r_01_MSTR
  bone2Name=spine_03
  swing1LimitDegrees=45
  swing1Motion=Free
  swing2LimitDegrees=45
  swing2Motion=Free
  twistLimitDegrees=45
  twistMotion=Free
 }
 {
  bone1Name=dyn_jacket_fr_l_01_MSTR
  bone2Name=spine_03
  swing1LimitDegrees=45
  swing1Motion=Free
  swing2LimitDegrees=45
  swing2Motion=Free
  twistLimitDegrees=45
  twistMotion=Free
 }
 {
  bone1Name=dyn_jacket_bk_l_01_MSTR
  bone2Name=spine_03
  swing1LimitDegrees=45
  swing1Motion=Free
  swing2LimitDegrees=45
  swing2Motion=Free
  twistLimitDegrees=45
  twistMotion=Free
 }
 {
  bone1Name=dyn_jacket_bkout_l_01_MSTR
  bone2Name=spine_03
  swing1LimitDegrees=45
  swing1Motion=Free
  swing2LimitDegrees=45
  swing2Motion=Free
  twistLimitDegrees=45
  twistMotion=Free
 }
 {
  bone1Name=dyn_jacket_l_01_MSTR
  bone2Name=spine_03
  swing1LimitDegrees=45
  swing1Motion=Free
  swing2LimitDegrees=45
  swing2Motion=Free
  twistLimitDegrees=45
  twistMotion=Free
 }
 {
  bone1Name=spine_01
  bone2Name=pelvis
  swing1LimitDegrees=45
  swing1Motion=Free
  swing2LimitDegrees=45
  swing2Motion=Free
  twistLimitDegrees=45
  twistMotion=Free
 }
 {
  bone1Name=spine_02
  bone2Name=spine_01
  swing1LimitDegrees=45
  swing1Motion=Free
  swing2LimitDegrees=45
  swing2Motion=Free
  twistLimitDegrees=45
  twistMotion=Free
 }
 {
  bone1Name=spine_03
  bone2Name=spine_02
  swing1LimitDegrees=45
  swing1Motion=Free
  swing2LimitDegrees=45
  swing2Motion=Free
  twistLimitDegrees=45
  twistMotion=Free
 }
 {
  bone1Name=thigh_l
  bone2Name=pelvis
  swing1LimitDegrees=45
  swing1Motion=Free
  swing2LimitDegrees=45
  swing2Motion=Free
  twistLimitDegrees=45
  twistMotion=Free
 }
 {
  bone1Name=thigh_r
  bone2Name=pelvis
  swing1LimitDegrees=45
  swing1Motion=Free
  swing2LimitDegrees=45
  swing2Motion=Free
  twistLimitDegrees=45
  twistMotion=Free
 }
]
