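# Physics asset for the necklace on the male medium Pencil jacket body, simulating dangling motion against the upper torso.
# Five necklace capsules hang from a small master body in a chain that forks at the second link, with tight twist and moderate swing limits.
# The master body attaches to spine_05, which links to spine_04, both with free angular motion.
# The two large spine capsules supply the torso volume.
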PhysicsAsset /BRCosmetics/Characters/Player/Male/Medium/Bodies/M_MED_Pencil_A/Jacket/Meshes/Pencil_Jacket_MedA_Ubdy_Jacket_Physics_NecklaceFE
bodies {
 dyn_necklace_01 [{shapeType=Capsule, length=4.8359, radius=3.8909, shapeName=dyn_necklace_01_capsule, center=<3 0 4>, rotation=<rot 0 0 90>}]
 dyn_necklace_02 [{shapeType=Capsule, length=3.6699, radius=5.0569, shapeName=dyn_necklace_02_capsule, center=<3 0 4>, rotation=<rot 0 0 90>}]
 dyn_necklace_03 [{shapeType=Capsule, length=4.8359, radius=3.8909, shapeName=dyn_necklace_03_capsule, center=<3 0 4>, rotation=<rot 0 0 90>}]
 dyn_necklace_04 [{shapeType=Capsule, length=3.6699, radius=5.0569, shapeName=dyn_necklace_04_capsule, center=<3 0 4>, rotation=<rot 0 0 90>}]
 dyn_necklace_05 [{shapeType=Capsule, length=4.4821, radius=4.2447, shapeName=dyn_necklace_05_capsule, center=<3 0 4>, rotation=<rot 0 0 90>}]
 dyn_necklace_MSTR [{shapeType=Capsule, length=0.505, radius=0.505, shapeName=dyn_necklace_MSTR_capsule, rotation=<rot 0 -90 0>}]
 spine_04 [{shapeType=Capsule, length=14.5668, radius=13.4586, shapeName=spine_04_capsule, center=<3.0923 -0.1572 0>}]
 spine_05 [{shapeType=Capsule, length=13.685, radius=10.8825, shapeName=spine_05_capsule}]
}
constraints [
 {bone1Name=dyn_necklace_01, bone2Name=dyn_necklace_MSTR, swing1LimitDegrees=30, swing2LimitDegrees=30, twistLimitDegrees=5}
 {bone1Name=dyn_necklace_02, bone2Name=dyn_necklace_01, swing1LimitDegrees=30, swing2LimitDegrees=30, twistLimitDegrees=5}
 {bone1Name=dyn_necklace_03, bone2Name=dyn_necklace_02, swing1LimitDegrees=30, swing2LimitDegrees=30, twistLimitDegrees=5}
 {bone1Name=dyn_necklace_04, bone2Name=dyn_necklace_03, swing1LimitDegrees=30, swing2LimitDegrees=30, twistLimitDegrees=5}
 {bone1Name=dyn_necklace_05, bone2Name=dyn_necklace_02, swing1LimitDegrees=30, swing2LimitDegrees=30, twistLimitDegrees=5}
 {
  bone1Name=dyn_necklace_MSTR
  bone2Name=spine_05
  swing1LimitDegrees=45
  swing1Motion=Free
  swing2LimitDegrees=45
  swing2Motion=Free
  twistLimitDegrees=45
  twistMotion=Free
 }
 {
  bone1Name=spine_05
  bone2Name=spine_04
  swing1LimitDegrees=45
  swing1Motion=Free
  swing2LimitDegrees=45
  swing2Motion=Free
  twistLimitDegrees=45
  twistMotion=Free
 }
]
