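# Post-process style AnimBlueprint for the Pencil male jacket cosmetic, adding secondary motion and pose fixes atop the parent mesh's copied pose.
# Deform and collar Control Rigs feed AnimDynamics necklace bones, then constraints and additive bone offsets adjust shirt, collar and badge.
# Crouch and rigid-body flags switch jacket branches using RigidBody simulation, constraints or movement-weighted Trail controllers.
# Final blends reposition collar bones for raised right arm, moving crouch and skydiving.
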AnimBlueprint /BRCosmetics/Characters/Player/Male/Medium/Bodies/M_MED_Pencil_A/Jacket/Meshes/Pencil_Jacket_MedA_Ubdy_AnimBP
target_skeleton=&/Pencil_Jacket_MedA_Ubdy_Skeleton
graphs [
 AnimGraph {nodes [
   Input Pose()
   n1 = Get RArmUp() -> (RArmUp:double)
   n2 = "Constraint - collar_bl_mstr (P)"(ComponentPose=Local To Component(LocalPose=Use cached pose 'PreCrouch'()), Alpha=@n1, ConstraintWeights_0=1) {
    alphaBoolBlend {blendOption=Linear}
    boneToModify {boneName=collar_bl_mstr}
    constraintSetup [{targetBone {boneName=neck_02}}]
    constraintWeights [1]
   }
   n3 = "Constraint - collar_b_mstr (P)"(ComponentPose=@n2, Alpha=@n1, ConstraintWeights_0=1) {
    alphaBoolBlend {blendOption=Linear}
    boneToModify {boneName=collar_b_mstr}
    constraintSetup [{targetBone {boneName=neck_02}}]
    constraintWeights [1]
   }
   n4 = "Constraint - collar_br_dyn (P)"(ComponentPose=@n3, Alpha=@n1, ConstraintWeights_0=1) {
    alphaBoolBlend {blendOption=Linear}
    boneToModify {boneName=collar_br_dyn}
    constraintSetup [{targetBone {boneName=neck_02}}]
    constraintWeights [1]
   }
   n5 = "Transform (Modify) Bone - Bone: collar_bl_mstr"(ComponentPose=@n4, Alpha=@n1, Translation="6.092,10.640,-0.910", Rotation="50,0,0", Scale="1,1,1") {
    rotationMode=BMM_Additive
    rotationSpace=BCS_BoneSpace
    scaleMode=BMM_Additive
    scaleSpace=BCS_BoneSpace
    translationMode=BMM_Additive
    translationSpace=BCS_BoneSpace
    alphaBoolBlend {blendOption=Linear}
    alphaScaleBiasClamp {bInterpResult=true, interpSpeedDecreasing=20, interpSpeedIncreasing=20}
    boneToModify {boneName=collar_bl_mstr}
    rotation {pitch=50}
    translation=<6.092 10.64 -0.91>
   }
   n6 = "Transform (Modify) Bone - Bone: collar_b_mstr"(ComponentPose=@n5, Alpha=@n1, Translation="5.365,0.039,1.014", Rotation="42.815,-0.183,0.026", Scale="1,1,1") {
    rotationMode=BMM_Additive
    rotationSpace=BCS_BoneSpace
    scaleMode=BMM_Additive
    scaleSpace=BCS_BoneSpace
    translationMode=BMM_Additive
    translationSpace=BCS_BoneSpace
    alphaBoolBlend {blendOption=Linear}
    alphaScaleBiasClamp {bInterpResult=true, interpSpeedDecreasing=20, interpSpeedIncreasing=20}
    boneToModify {boneName=collar_b_mstr}
    rotation=<rot 42.815 -0.183 0.026>
    translation=<5.365 0.039 1.014>
   }
   n7 = "Transform (Modify) Bone - Bone: collar_br_mstr"(ComponentPose=@n6, Alpha=@n1, Translation="10.291,0.000,-2.169", Rotation="50,0,0", Scale="1,1,1") {
    rotationMode=BMM_Additive
    rotationSpace=BCS_BoneSpace
    scaleMode=BMM_Additive
    scaleSpace=BCS_BoneSpace
    translationMode=BMM_Additive
    translationSpace=BCS_BoneSpace
    alphaBoolBlend {blendOption=Linear}
    alphaScaleBiasClamp {bInterpResult=true, interpSpeedDecreasing=20, interpSpeedIncreasing=20}
    boneToModify {boneName=collar_br_mstr}
    rotation {pitch=50}
    translation {x=10.291, z=-2.169}
   }
   n8 = "Transform (Modify) Bone - Bone: collar_fl_mstr"(ComponentPose=@n7, Alpha=@n1, Translation="-0.893,5.396,-3.154", Rotation="-72.341,0.000,0.000", Scale="1,1,1") {
    rotationMode=BMM_Additive
    rotationSpace=BCS_BoneSpace
    scaleMode=BMM_Additive
    scaleSpace=BCS_BoneSpace
    translationMode=BMM_Additive
    translationSpace=BCS_BoneSpace
    alphaBoolBlend {blendOption=Linear}
    alphaScaleBiasClamp {bInterpResult=true, interpSpeedDecreasing=20, interpSpeedIncreasing=20}
    boneToModify {boneName=collar_fl_mstr}
    rotation=<rot -72.341 0 0>
    translation=<-0.893 5.396 -3.154>
   }
   n9 = "Transform (Modify) Bone - Bone: collar_l_mstr"(ComponentPose=@n8, Alpha=@n1, Translation="8.287,0.369,0.312", Rotation="-0.000,0.000,-90.472", Scale="1,1,1") {
    rotationMode=BMM_Additive
    rotationSpace=BCS_BoneSpace
    scaleMode=BMM_Additive
    scaleSpace=BCS_BoneSpace
    translationMode=BMM_Additive
    translationSpace=BCS_BoneSpace
    alphaBoolBlend {blendOption=Linear}
    alphaScaleBiasClamp {bInterpResult=true, interpSpeedDecreasing=20, interpSpeedIncreasing=20}
    boneToModify {boneName=collar_l_mstr}
    rotation=<rot 0 0 -90.472>
    translation=<8.287 0.369 0.312>
   }
   n10 = "Transform (Modify) Bone - Bone: spine_05"(ComponentPose=@n9, Alpha=@n1, Translation="0,0,0", Rotation="0,0,0", Scale="1,1,1") {
    rotationMode=BMM_Additive
    rotationSpace=BCS_BoneSpace
    scaleMode=BMM_Additive
    scaleSpace=BCS_BoneSpace
    translationMode=BMM_Additive
    translationSpace=BCS_BoneSpace
    alphaBoolBlend {blendOption=Linear}
    alphaScaleBiasClamp {bInterpResult=true, interpSpeedDecreasing=20, interpSpeedIncreasing=20}
    boneToModify {boneName=spine_05}
   }
   n11 = Component To Local(ComponentPose=@n10)
   n12 = Use cached pose 'PreCrouch'()
   n13 = Get Moving Crouch() -> (Moving Crouch:bool)
   n14 = "Constraint - collar_bl_mstr (P)"(ComponentPose=Local To Component(LocalPose=Use cached pose 'PreCrouch'()), Alpha=1, ConstraintWeights_0=1) {
    alphaBoolBlend {blendOption=Linear}
    boneToModify {boneName=collar_bl_mstr}
    constraintSetup [{targetBone {boneName=head}}]
    constraintWeights [1]
   }
   n15 = "Constraint - collar_b_mstr (P)"(ComponentPose=@n14, Alpha=1, ConstraintWeights_0=1) {
    alphaBoolBlend {blendOption=Linear}
    boneToModify {boneName=collar_b_mstr}
    constraintSetup [{targetBone {boneName=head}}]
    constraintWeights [1]
   }
   n16 = "Constraint - collar_br_mstr (P)"(ComponentPose=@n15, Alpha=1, ConstraintWeights_0=1) {
    alphaBoolBlend {blendOption=Linear}
    boneToModify {boneName=collar_br_mstr}
    constraintSetup [{targetBone {boneName=head}}]
    constraintWeights [1]
   }
   n17 = "Transform (Modify) Bone - Bone: collar_b_mstr"(ComponentPose=@n16, Alpha=1, Translation="2,0,5", Rotation="0,0,0", Scale="0.8,1,1") {
    rotationMode=BMM_Additive
    rotationSpace=BCS_BoneSpace
    scaleMode=BMM_Additive
    scaleSpace=BCS_BoneSpace
    translationMode=BMM_Additive
    translationSpace=BCS_BoneSpace
    alphaBoolBlend {blendOption=Linear}
    alphaScaleBiasClamp {bInterpResult=true, interpSpeedDecreasing=20, interpSpeedIncreasing=20}
    boneToModify {boneName=collar_b_mstr}
    scale {x=0.8}
    translation {x=2, z=5}
   }
   Output Pose(Result=Blend Poses by bool(
     BlendPose_0=Component To Local(ComponentPose=@n17)
     BlendPose_1=Blend Poses by bool(BlendPose_0=@n11, BlendPose_1=@n12, BlendTime_0=0.2, bActiveValue=@n13) {blendType=Linear}
     BlendTime_0=0
     bActiveValue=Get bIsSkyDiving()
    ) {blendType=Linear}) {layerGroup=None}
   n18 = Control Rig(Source=Copy Pose From Mesh() {bCopyCurves=true, bUseAttachedParent=true}) {
    alphaBoolBlend {blendOption=Linear}
    controlRigAssetReference {blueprintRigClass=&/Pencil_Jacket_MedA_Ubdy_DeformRig_Master_ControlRig.Pencil_Jacket_MedA_Ubdy_DeformRig_Master_ControlRig_C}
   }
   n19 = Control Rig!(
    Source=@n18
    bAlphaBoolEnabled:bool=True
    PencilApple:bool=Get PencilApple() -> (PencilApple:bool)
    Crouching:bool=Get newCrouching() -> (newCrouching:bool)
    PencilGrape:bool=Get PencilGrape() -> (PencilGrape:bool)
    PencilRaspberry:bool=Get PencilRaspberry() -> (PencilRaspberry:bool)
   ) -> (Pose) {
    alphaInputType=Bool
    alphaBoolBlend {blendOption=Linear}
    controlRigAssetReference {blueprintRigClass=&/Pencil_Jacket_MedA_Ubdy_Collar_CtrlRig.Pencil_Jacket_MedA_Ubdy_Collar_CtrlRig_C}
   }
   n20 = Anim Dynamics - Bone: dyn_necklace_01(ComponentPose=Local To Component(LocalPose=@n19), Alpha=0.5, ExternalForce="0,0,0") {
    angularDampingOverride=1
    angularSpringConstant=1000
    bAngularSpring=true
    bGravityOverrideInSimSpace=true
    bOverrideAngularDamping=true
    bUseSphericalLimits=true
    gravityScale=0.35
    simulationSpace=RootRelative
    alphaBoolBlend {blendOption=Linear}
    alphaScaleBiasClamp {scale=0.7}
    boundBone {boneName=dyn_necklace_01}
    componentLinearAccScale=<0.4 0.4 0.46>
    componentLinearVelScale=<0.4 0.4 0.46>
    physicsBodyDefinitions [{
      sphereCollisionRadius=0
      boundBone {boneName=dyn_necklace_01}
      boxExtents=<10 20 20>
      constraintSetup {angularLimitsMax=<10 20 20>, angularLimitsMin=<-10 -20 -40>, angularTarget=<0 0 0>}
      localJointOffset=<5 0 0>
     }]
    sphericalLimits [
     {limitRadius=20, drivingBone {boneName=spine_05}, sphereLocalOffset=<0 -10 0>}
     {limitRadius=20, drivingBone {boneName=spine_05}, sphereLocalOffset=<0 10 26>}
     {limitRadius=12, drivingBone {boneName=spine_05}, sphereLocalOffset=<0 10 -20>}
    ]
   }
   n21 = Anim Dynamics - Bone: dyn_necklace_02(ComponentPose=@n20, Alpha=0.5, ExternalForce="0,0,0") {
    angularDampingOverride=0.999
    angularSpringConstant=1000
    bAngularSpring=true
    bGravityOverrideInSimSpace=true
    bOverrideAngularDamping=true
    bUseSphericalLimits=true
    gravityScale=0.35
    simulationSpace=RootRelative
    alphaBoolBlend {blendOption=Linear}
    alphaScaleBiasClamp {scale=0.7}
    boundBone {boneName=dyn_necklace_02}
    componentLinearAccScale=<0.4 0.4 0.46>
    componentLinearVelScale=<0.4 0.4 0.46>
    physicsBodyDefinitions [{
      sphereCollisionRadius=0
      boundBone {boneName=dyn_necklace_02}
      boxExtents=<8 15 15>
      constraintSetup {angularLimitsMax=<10 20 20>, angularLimitsMin=<-10 -20 -40>, angularTarget=<0 0 0>}
      localJointOffset=<5 0 0>
     }]
    sphericalLimits [
     {limitRadius=20, drivingBone {boneName=spine_05}, sphereLocalOffset=<0 -10 0>}
     {limitRadius=20, drivingBone {boneName=spine_05}, sphereLocalOffset=<0 10 26>}
     {limitRadius=12, drivingBone {boneName=spine_05}, sphereLocalOffset=<0 10 -21.5>}
    ]
   }
   n22 = Anim Dynamics - Bone: dyn_necklace_03(ComponentPose=@n21, Alpha=0.5, ExternalForce="0,0,0") {
    angularDampingOverride=1
    angularSpringConstant=1000
    bAngularSpring=true
    bGravityOverrideInSimSpace=true
    bOverrideAngularDamping=true
    bUseSphericalLimits=true
    gravityScale=0.35
    simulationSpace=RootRelative
    alphaBoolBlend {blendOption=Linear}
    alphaScaleBiasClamp {scale=0.7}
    boundBone {boneName=dyn_necklace_03}
    componentLinearAccScale=<0.4 0.4 0.46>
    componentLinearVelScale=<0.4 0.4 0.46>
    physicsBodyDefinitions [{
      sphereCollisionRadius=0
      boundBone {boneName=dyn_necklace_03}
      boxExtents=<10 20 20>
      constraintSetup {angularLimitsMax=<10 20 20>, angularLimitsMin=<-10 -20 -40>, angularTarget=<0 0 0>}
      localJointOffset=<5 0 0>
     }]
    sphericalLimits [
     {limitRadius=20, drivingBone {boneName=spine_05}, sphereLocalOffset=<0 -5 0>}
     {limitRadius=20, drivingBone {boneName=spine_05}, sphereLocalOffset=<0 10 26>}
     {limitRadius=12, drivingBone {boneName=spine_05}, sphereLocalOffset=<0 10 -22>}
    ]
   }
   n23 = Anim Dynamics - Bone: dyn_necklace_04(ComponentPose=@n22, Alpha=0.5, ExternalForce="0,0,0") {
    angularDampingOverride=1
    angularSpringConstant=1000
    bAngularSpring=true
    bGravityOverrideInSimSpace=true
    bOverrideAngularDamping=true
    bUseSphericalLimits=true
    gravityScale=0.35
    simulationSpace=RootRelative
    alphaBoolBlend {blendOption=Linear}
    alphaScaleBiasClamp {scale=0.7}
    boundBone {boneName=dyn_necklace_04}
    componentLinearAccScale=<0.4 0.4 0.46>
    componentLinearVelScale=<0.4 0.4 0.46>
    physicsBodyDefinitions [{
      sphereCollisionRadius=0
      boundBone {boneName=dyn_necklace_04}
      boxExtents=<5 10 10>
      constraintSetup {angularLimitsMax=<10 20 20>, angularLimitsMin=<-10 -20 -40>, angularTarget=<0 0 0>}
      localJointOffset=<5 0 0>
     }]
    sphericalLimits [
     {limitRadius=20, drivingBone {boneName=spine_05}, sphereLocalOffset=<0 -5 0>}
     {limitRadius=20, drivingBone {boneName=spine_05}, sphereLocalOffset=<0 10 26>}
     {limitRadius=12, drivingBone {boneName=spine_05}, sphereLocalOffset=<0 10 -22.5>}
     {limitRadius=19, drivingBone {boneName=spine_04}, sphereLocalOffset=<0 0 0>}
    ]
   }
   n24 = Anim Dynamics - Bone: dyn_necklace_05(ComponentPose=@n23, Alpha=0.5, ExternalForce="0,0,0") {
    angularDampingOverride=0.9999
    angularSpringConstant=500
    bAngularSpring=true
    bGravityOverrideInSimSpace=true
    bOverrideAngularDamping=true
    bUseSphericalLimits=true
    gravityScale=0.35
    simulationSpace=RootRelative
    alphaBoolBlend {blendOption=Linear}
    alphaScaleBiasClamp {scale=0.7}
    boundBone {boneName=dyn_necklace_05}
    componentLinearAccScale=<0.4 0.4 0.46>
    componentLinearVelScale=<0.4 0.4 0.46>
    physicsBodyDefinitions [{
      sphereCollisionRadius=0
      boundBone {boneName=dyn_necklace_05}
      boxExtents=<5 10 10>
      constraintSetup {angularLimitsMax=<10 20 40>, angularLimitsMin=<-10 -20 -40>, angularTarget=<0 0 0>}
      localJointOffset=<5 0 0>
     }]
    sphericalLimits [
     {limitRadius=20, drivingBone {boneName=spine_05}, sphereLocalOffset=<0 -5 0>}
     {limitRadius=20, drivingBone {boneName=spine_05}, sphereLocalOffset=<0 10 26>}
     {limitRadius=20, drivingBone {boneName=spine_05}, sphereLocalOffset=<0 10 -23>}
    ]
   }
   n25 = "Constraint - dyn_shirt_l_MSTR (PPPPPPPP)"!(
    ComponentPose=@n24
    Alpha=1
    ConstraintWeights_0=0.3784
    ConstraintWeights_1:float=0.3892
    ConstraintWeights_2:float=0.0024
    ConstraintWeights_3:float=0.1899
    ConstraintWeights_4:float=0.018
    ConstraintWeights_5:float=0.0035
    ConstraintWeights_6:float=0.0046
    ConstraintWeights_7:float=0.014
   ) -> (Pose) {
    alphaBoolBlend {blendOption=Linear}
    boneToModify {boneName=dyn_shirt_l_MSTR}
    constraintSetup [
     {targetBone {boneName=pelvis}}
     {targetBone {boneName=thigh_twist_01_l}}
     {targetBone {boneName=deform_glute_l}}
     {targetBone {boneName=deform_groin_l}}
     {targetBone {boneName=thigh_twist_01_r}}
     {targetBone {boneName=spine_02}}
     {targetBone {boneName=thigh_l}}
     {targetBone {boneName=deform_groin_r}}
    ]
    constraintWeights [1,1,1,1,1,1,1,1]
   }
   n26 = "Constraint - dyn_shirt_r_MSTR (PPPPPPPP)"!(
    ComponentPose=@n25
    Alpha=1
    ConstraintWeights_0=0.414
    ConstraintWeights_1:float=0.0448
    ConstraintWeights_2:float=0.0085
    ConstraintWeights_3:float=0.0243
    ConstraintWeights_4:float=0.3228
    ConstraintWeights_5:float=0.0047
    ConstraintWeights_6:float=0.0094
    ConstraintWeights_7:float=0.1715
   ) -> (Pose) {
    alphaBoolBlend {blendOption=Linear}
    boneToModify {boneName=dyn_shirt_r_MSTR}
    constraintSetup [
     {targetBone {boneName=pelvis}}
     {targetBone {boneName=thigh_twist_01_l}}
     {targetBone {boneName=thigh_r}}
     {targetBone {boneName=deform_groin_l}}
     {targetBone {boneName=thigh_twist_01_r}}
     {targetBone {boneName=spine_01}}
     {targetBone {boneName=spine_02}}
     {targetBone {boneName=deform_groin_r}}
    ]
    constraintWeights [1,1,1,1,1,1,1,1]
   }
   n27 = "Transform (Modify) Bone - Bone: dyn_shirt_l_MSTR"!(ComponentPose=@n26, bAlphaBoolEnabled:bool=True, Translation="-0.622,1.876,-0.595", Rotation="0.002,0.103,0.010", Scale="1,1,1") -> (Pose) {
    alphaInputType=Bool
    rotationMode=BMM_Additive
    rotationSpace=BCS_BoneSpace
    scaleMode=BMM_Additive
    scaleSpace=BCS_BoneSpace
    translationMode=BMM_Additive
    translationSpace=BCS_BoneSpace
    alphaBoolBlend {blendInTime=0.1, blendOption=Linear, blendOutTime=0.1}
    boneToModify {boneName=dyn_shirt_l_MSTR}
    rotation=<rot 0.002 0.103 0.01>
    translation=<-0.622 1.876 -0.595>
   }
   n28 = "Transform (Modify) Bone - Bone: dyn_shirt_r_MSTR"!(ComponentPose=@n27, bAlphaBoolEnabled:bool=True, Translation="-0.000,-0.988,-0.258", Rotation="0.008,-5.373,-0.003", Scale="1,1,1") -> (Pose) {
    alphaInputType=Bool
    rotationMode=BMM_Additive
    rotationSpace=BCS_BoneSpace
    scaleMode=BMM_Additive
    scaleSpace=BCS_BoneSpace
    translationMode=BMM_Additive
    translationSpace=BCS_BoneSpace
    alphaBoolBlend {blendInTime=0.1, blendOption=Linear, blendOutTime=0.1}
    boneToModify {boneName=dyn_shirt_r_MSTR}
    rotation=<rot 0.008 -5.373 -0.003>
    translation {y=-0.988, z=-0.258}
   }
   n29 = "Constraint - dyn_shirt_l_MSTR (R)"!(ComponentPose=@n28, bAlphaBoolEnabled:bool=True, ConstraintWeights_0=1) -> (Pose) {
    alphaInputType=Bool
    alphaBoolBlend {blendOption=Linear}
    boneToModify {boneName=dyn_shirt_l_MSTR}
    constraintSetup [{transformType=Rotation, targetBone {boneName=thigh_twist_01_l}}]
    constraintWeights [1]
   }
   n30 = "Constraint - dyn_shirt_r_MSTR (R)"!(ComponentPose=@n29, bAlphaBoolEnabled:bool=True, ConstraintWeights_0=1) -> (Pose) {
    alphaInputType=Bool
    alphaBoolBlend {blendOption=Linear}
    boneToModify {boneName=dyn_shirt_r_MSTR}
    constraintSetup [{transformType=Rotation, targetBone {boneName=thigh_twist_01_r}}]
    constraintWeights [1]
   }
   n31 = "Transform (Modify) Bone - Bone: dyn_shirt_l_MSTR"(ComponentPose=@n30, bAlphaBoolEnabled=True, Translation="-5.472,-0.832,-2.502", Rotation="-19.219,-9.928,-31.963", Scale="1,1,1") {
    alphaInputType=Bool
    rotationMode=BMM_Additive
    rotationSpace=BCS_BoneSpace
    scaleMode=BMM_Additive
    scaleSpace=BCS_BoneSpace
    translationMode=BMM_Additive
    translationSpace=BCS_BoneSpace
    alphaBoolBlend {blendInTime=0.1, blendOption=Linear, blendOutTime=0.1}
    boneToModify {boneName=dyn_shirt_l_MSTR}
    rotation=<rot -19.219 -9.928 -31.963>
    translation=<-5.472 -0.832 -2.502>
   }
   n32 = "Transform (Modify) Bone - Bone: dyn_shirt_r_MSTR"(ComponentPose=@n31, bAlphaBoolEnabled=True, Translation="-2.514,-1.5,-0.378", Rotation="-13.880,-0.013,-0.031", Scale="1,1,1") {
    alphaInputType=Bool
    rotationMode=BMM_Additive
    rotationSpace=BCS_BoneSpace
    scaleMode=BMM_Additive
    scaleSpace=BCS_BoneSpace
    translationMode=BMM_Additive
    translationSpace=BCS_BoneSpace
    alphaBoolBlend {blendInTime=0.1, blendOption=Linear, blendOutTime=0.1}
    boneToModify {boneName=dyn_shirt_r_MSTR}
    rotation=<rot -13.88 -0.013 -0.031>
    translation=<-2.514 -1.5 -0.378>
   }
   n33 = "Transform (Modify) Bone - Bone: dyn_shirt_l_MSTR"(ComponentPose=@n32, bAlphaBoolEnabled=True, Translation="1.191,0.197,-1.116", Rotation="6.745,5.311,35.170", Scale="1,1,1.5") {
    alphaInputType=Bool
    rotationMode=BMM_Additive
    rotationSpace=BCS_BoneSpace
    scaleMode=BMM_Additive
    scaleSpace=BCS_BoneSpace
    translationMode=BMM_Additive
    translationSpace=BCS_BoneSpace
    alphaBoolBlend {blendInTime=0.1, blendOption=Linear, blendOutTime=0.1}
    boneToModify {boneName=dyn_shirt_l_MSTR}
    rotation=<rot 6.745 5.311 35.17>
    scale {z=1.5}
    translation=<1.191 0.197 -1.116>
   }
   n34 = "Transform (Modify) Bone - Bone: dyn_shirt_r_MSTR"!(
    ComponentPose=@n33
    Alpha:float=Get thigh_rot_otr_r() -> (thigh_rot_otr_r:double)
    Translation="-1.395,-0.845,0.253"
    Rotation="-23.023,5.785,11.528"
    Scale="1,1,1"
   ) -> (Pose) {
    rotationMode=BMM_Additive
    rotationSpace=BCS_BoneSpace
    scaleMode=BMM_Additive
    scaleSpace=BCS_BoneSpace
    translationMode=BMM_Additive
    translationSpace=BCS_BoneSpace
    alphaBoolBlend {blendInTime=0.1, blendOption=Linear, blendOutTime=0.1}
    boneToModify {boneName=dyn_shirt_r_MSTR}
    rotation=<rot -23.023 5.785 11.528>
    translation=<-1.395 -0.845 0.253>
   }
   n35 = "Transform (Modify) Bone - Bone: dyn_shirt_r_MSTR"!(ComponentPose=@n34, bAlphaBoolEnabled:bool=True, Translation="3.24,0,7.928", Rotation="5.163,28.454,-7.592", Scale="1,1,1") -> (Pose) {
    alphaInputType=Bool
    rotationMode=BMM_Additive
    rotationSpace=BCS_BoneSpace
    scaleMode=BMM_Additive
    scaleSpace=BCS_BoneSpace
    translationMode=BMM_Additive
    translationSpace=BCS_BoneSpace
    alphaBoolBlend {blendInTime=0.1, blendOption=Linear, blendOutTime=0.1}
    boneToModify {boneName=dyn_shirt_r_MSTR}
    rotation=<rot 5.163 28.454 -7.592>
    translation {x=3.24, z=7.928}
   }
   n36 = "Transform (Modify) Bone - Bone: collar_br_mstr"!(ComponentPose=@n35, bAlphaBoolEnabled:bool=True, Translation="0,0,0", Rotation="15,0,0", Scale="1,1,1") -> (Pose) {
    alphaInputType=Bool
    rotationMode=BMM_Additive
    rotationSpace=BCS_BoneSpace
    scaleMode=BMM_Additive
    scaleSpace=BCS_BoneSpace
    translationMode=BMM_Additive
    translationSpace=BCS_BoneSpace
    alphaBoolBlend {blendInTime=0.1, blendOption=Linear, blendOutTime=0.1}
    boneToModify {boneName=collar_br_mstr}
    rotation {pitch=15}
   }
   Save cached pose 'DeformPost'(Pose=Component To Local(ComponentPose=@n36)) {cacheName=DeformPost}
   n37 = "dyn_jacket_fr_r_01_MSTR.rotateY = thigh_r.rotateZ * -1 - Bone Driven Controller"!(ComponentPose:Component Space Pose Link Structure=Local To Component(LocalPose=Use cached pose 'DeformPost'()), Alpha:float=0.6) -> (Pose:Component Space Pose Link Structure) {
    is=title AnimGraphNode_BoneDrivenController
    bAffectTargetRotationY=true
    multiplier=-1
    sourceComponent=RotationZ
    alphaBoolBlend {blendOption=Linear}
    sourceBone {boneName=thigh_r}
    targetBone {boneName=dyn_jacket_fr_r_01_MSTR}
   }
   n38 = "Constraint - dyn_jacket_fr_l_01_MSTR (PPPPPPPP)"!(
    ComponentPose=@n37
    bAlphaBoolEnabled:bool=True
    ConstraintWeights_0=0.2781
    ConstraintWeights_1:float=0.5714
    ConstraintWeights_2:float=0.0598
    ConstraintWeights_3:float=0.0778
    ConstraintWeights_4:float=0.0005
    ConstraintWeights_5:float=0.0003
    ConstraintWeights_6:float=0.0006
    ConstraintWeights_7:float=0.0116
   ) -> (Pose) {
    alphaInputType=Bool
    alphaBoolBlend {blendOption=Linear}
    boneToModify {boneName=dyn_jacket_fr_l_01_MSTR}
    constraintSetup [
     {targetBone {boneName=pelvis}}
     {targetBone {boneName=thigh_twist_01_l}}
     {targetBone {boneName=deform_glute_l}}
     {targetBone {boneName=deform_groin_l}}
     {targetBone {boneName=thigh_twist_01_r}}
     {targetBone {boneName=spine_01}}
     {targetBone {boneName=spine_02}}
     {targetBone {boneName=thigh_l}}
    ]
    constraintWeights [1,1,1,1,1,1,1,1]
   }
   n39 = "Transform (Modify) Bone - Bone: dyn_jacket_fr_l_01_MSTR"!(ComponentPose=@n38, bAlphaBoolEnabled:bool=True, Translation="21.780,-0.508,19.967", Rotation="-39.286,-60.766,1.847", Scale="0.5,1,1") -> (Pose) {
    alphaInputType=Bool
    rotationMode=BMM_Additive
    rotationSpace=BCS_BoneSpace
    scaleMode=BMM_Additive
    scaleSpace=BCS_BoneSpace
    translationMode=BMM_Additive
    translationSpace=BCS_BoneSpace
    alphaBoolBlend {blendOption=Linear}
    boneToModify {boneName=dyn_jacket_fr_l_01_MSTR}
    rotation=<rot -39.286 -60.766 1.847>
    scale {x=0.5}
    translation=<21.78 -0.508 19.967>
   }
   n40 = "Transform (Modify) Bone - Bone: dyn_jacket_fr_l_01_MSTR"(
    ComponentPose=Local To Component(LocalPose=Use cached pose 'DeformPost'())
    bAlphaBoolEnabled=True
    Translation="0,0,0"
    Rotation="8.889,0.004,-0.011"
    Scale="1,1,1"
   ) {
    alphaInputType=Bool
    rotationMode=BMM_Additive
    rotationSpace=BCS_BoneSpace
    scaleMode=BMM_Additive
    scaleSpace=BCS_BoneSpace
    translationMode=BMM_Additive
    translationSpace=BCS_BoneSpace
    alphaBoolBlend {blendOption=Linear}
    boneToModify {boneName=dyn_jacket_fr_l_01_MSTR}
    rotation=<rot 8.889 0.004 -0.011>
   }
   n41 = Get Bwd Movement() -> (Bwd Movement:double)
   n42 = Trail controller - Bone: dyn_jacket_bk_02!(ComponentPose=@n40, Alpha=@n41, FakeVelocity:Vector="0,0,0") -> (Pose) {
    bActorSpaceFakeVel=true
    bShowTrailLocation=true
    debugLifeTime=0.1
    maxDeltaTime=0.0333
    relaxationSpeedScale=3
    alphaScaleBiasClamp {bInterpResult=true}
    baseJoint {boneName=root}
    planarLimits [
     {drivingBone {boneName=root}, planeTransform {location=<0 0 10>, rotation=<rot 0 0 0>, scale=<1 1 1>}}
     {drivingBone {boneName=spine_03}, planeTransform {location=<0 -20 0>, rotation=<rot 0 15 -89.9999>, scale=<1 1 1>}}
    ]
    rotationLimits [{limitMax=<180 180 180>}, {limitMax=<180 180 180>}]
    rotationOffsets [<0 0 0>, <0 0 0>]
    trailBone {boneName=dyn_jacket_bk_02}
    trailRelaxationSpeed {editorCurveData {keys [{arriveTangent=-72.0739, interpMode=RCIM_Cubic, leaveTangent=-72.0739, tangentMode=RCTM_User, time=0, value=35}, <at 1 5 RCIM_Cubic>]}}
   }
   n43 = Trail controller - Bone: dyn_jacket_bk_l_02!(ComponentPose=@n42, Alpha=@n41, FakeVelocity:Vector="0,0,0") -> (Pose) {
    bActorSpaceFakeVel=true
    bShowTrailLocation=true
    debugLifeTime=0.1
    maxDeltaTime=0.0333
    relaxationSpeedScale=3
    alphaScaleBiasClamp {bInterpResult=true}
    baseJoint {boneName=root}
    planarLimits [
     {drivingBone {boneName=root}, planeTransform {location=<0 0 10>, rotation=<rot 0 0 0>, scale=<1 1 1>}}
     {drivingBone {boneName=spine_03}, planeTransform {location=<0 -20 0>, rotation=<rot 0 15 -89.9999>, scale=<1 1 1>}}
    ]
    rotationLimits [{limitMax=<180 180 180>}, {limitMax=<180 180 180>}]
    rotationOffsets [<0 0 0>, <0 0 0>]
    trailBone {boneName=dyn_jacket_bk_l_02}
    trailRelaxationSpeed {editorCurveData {keys [{arriveTangent=-72.0739, interpMode=RCIM_Cubic, leaveTangent=-72.0739, tangentMode=RCTM_User, time=0, value=35}, <at 1 5 RCIM_Cubic>]}}
   }
   n44 = Trail controller - Bone: dyn_jacket_bk_r_02!(ComponentPose=@n43, Alpha=@n41, FakeVelocity:Vector="0,0,0") -> (Pose) {
    bActorSpaceFakeVel=true
    bShowTrailLocation=true
    debugLifeTime=0.1
    maxDeltaTime=0.0333
    relaxationSpeedScale=3
    alphaScaleBiasClamp {bInterpResult=true}
    baseJoint {boneName=root}
    planarLimits [
     {drivingBone {boneName=root}, planeTransform {location=<0 0 10>, rotation=<rot 0 0 0>, scale=<1 1 1>}}
     {drivingBone {boneName=spine_03}, planeTransform {location=<0 -20 0>, rotation=<rot 0 15 -89.9999>, scale=<1 1 1>}}
    ]
    rotationLimits [{limitMax=<180 180 180>}, {limitMax=<180 180 180>}]
    rotationOffsets [<0 0 0>, <0 0 0>]
    trailBone {boneName=dyn_jacket_bk_r_02}
    trailRelaxationSpeed {editorCurveData {keys [{arriveTangent=-72.0739, interpMode=RCIM_Cubic, leaveTangent=-72.0739, tangentMode=RCTM_User, time=0, value=35}, <at 1 5 RCIM_Cubic>]}}
   }
   n45 = Get Right Movement() -> (Right Movement:double)
   n46 = Trail controller - Bone: dyn_jacket_bkout_r_02!(
    ComponentPose=@n44
    Alpha=FInterpEaseInOut!(A:double=@n41, B:double=@n45, Alpha:double=1, Exponent:double=0) -> (ReturnValue:double)
    FakeVelocity:Vector="0,0,0"
   ) -> (Pose) {
    bActorSpaceFakeVel=true
    bShowTrailLocation=true
    debugLifeTime=0.1
    maxDeltaTime=0.0333
    relaxationSpeedScale=3
    alphaScaleBiasClamp {bInterpResult=true}
    baseJoint {boneName=root}
    planarLimits [
     {drivingBone {boneName=root}, planeTransform {location=<0 0 10>, rotation=<rot 0 0 0>, scale=<1 1 1>}}
     {drivingBone {boneName=spine_03}, planeTransform {location=<0 -20 0>, rotation=<rot 0 15 -89.9999>, scale=<1 1 1>}}
    ]
    rotationLimits [{limitMax=<180 180 180>}, {limitMax=<180 180 180>}]
    rotationOffsets [<0 0 0>, <0 0 0>]
    trailBone {boneName=dyn_jacket_bkout_r_02}
    trailRelaxationSpeed {editorCurveData {keys [{arriveTangent=-72.0739, interpMode=RCIM_Cubic, leaveTangent=-72.0739, tangentMode=RCTM_User, time=0, value=35}, <at 1 5 RCIM_Cubic>]}}
   }
   n47 = Get Left Movement() -> (Left Movement:double)
   n48 = Trail controller - Bone: dyn_jacket_bkout_l_02!(ComponentPose=@n46, Alpha=FInterpEaseInOut(A=@n41, B=@n47, Alpha=1, Exponent=0), FakeVelocity:Vector="0,0,0") -> (Pose) {
    bActorSpaceFakeVel=true
    bShowTrailLocation=true
    debugLifeTime=0.1
    maxDeltaTime=0.0333
    relaxationSpeedScale=3
    alphaScaleBiasClamp {bInterpResult=true}
    baseJoint {boneName=root}
    planarLimits [
     {drivingBone {boneName=root}, planeTransform {location=<0 0 10>, rotation=<rot 0 0 0>, scale=<1 1 1>}}
     {drivingBone {boneName=spine_03}, planeTransform {location=<0 -20 0>, rotation=<rot 0 15 -89.9999>, scale=<1 1 1>}}
    ]
    rotationLimits [{limitMax=<180 180 180>}, {limitMax=<180 180 180>}]
    rotationOffsets [<0 0 0>, <0 0 0>]
    trailBone {boneName=dyn_jacket_bkout_l_02}
    trailRelaxationSpeed {editorCurveData {keys [{arriveTangent=-72.0739, interpMode=RCIM_Cubic, leaveTangent=-72.0739, tangentMode=RCTM_User, time=0, value=35}, <at 1 5 RCIM_Cubic>]}}
   }
   n49 = Trail controller - Bone: dyn_jacket_l_02!(ComponentPose=@n48, Alpha=@n47, FakeVelocity:Vector="0,0,0") -> (Pose) {
    bActorSpaceFakeVel=true
    bShowTrailLocation=true
    debugLifeTime=0.1
    maxDeltaTime=0.0333
    relaxationSpeedScale=3
    alphaScaleBiasClamp {bInterpResult=true}
    baseJoint {boneName=root}
    planarLimits [
     {drivingBone {boneName=root}, planeTransform {location=<0 0 10>, rotation=<rot 0 0 0>, scale=<1 1 1>}}
     {drivingBone {boneName=spine_03}, planeTransform {location=<0 -20 0>, rotation=<rot 0 15 -89.9999>, scale=<1 1 1>}}
    ]
    rotationLimits [{limitMax=<180 180 180>}, {limitMax=<180 180 180>}]
    rotationOffsets [<0 0 0>, <0 0 0>]
    trailBone {boneName=dyn_jacket_l_02}
    trailRelaxationSpeed {editorCurveData {keys [{arriveTangent=-72.0739, interpMode=RCIM_Cubic, leaveTangent=-72.0739, tangentMode=RCTM_User, time=0, value=35}, <at 1 5 RCIM_Cubic>]}}
   }
   n50 = Trail controller - Bone: dyn_jacket_fr_l_02!(ComponentPose=@n49, Alpha=@n47, FakeVelocity:Vector="0,0,0") -> (Pose) {
    bActorSpaceFakeVel=true
    bShowTrailLocation=true
    debugLifeTime=0.1
    maxDeltaTime=0.0333
    relaxationSpeedScale=3
    alphaScaleBiasClamp {bInterpResult=true}
    baseJoint {boneName=root}
    planarLimits [
     {drivingBone {boneName=root}, planeTransform {location=<0 0 10>, rotation=<rot 0 0 0>, scale=<1 1 1>}}
     {drivingBone {boneName=spine_03}, planeTransform {location=<0 -20 0>, rotation=<rot 0 15 -89.9999>, scale=<1 1 1>}}
    ]
    rotationLimits [{limitMax=<180 180 180>}, {limitMax=<180 180 180>}]
    rotationOffsets [<0 0 0>, <0 0 0>]
    trailBone {boneName=dyn_jacket_fr_l_02}
    trailRelaxationSpeed {editorCurveData {keys [{arriveTangent=-72.0739, interpMode=RCIM_Cubic, leaveTangent=-72.0739, tangentMode=RCTM_User, time=0, value=35}, <at 1 5 RCIM_Cubic>]}}
   }
   n51 = Trail controller - Bone: dyn_jacket_r_02!(ComponentPose=@n50, Alpha=@n45, FakeVelocity:Vector="0,0,0") -> (Pose) {
    bActorSpaceFakeVel=true
    bShowTrailLocation=true
    debugLifeTime=0.1
    maxDeltaTime=0.0333
    relaxationSpeedScale=3
    alphaScaleBiasClamp {bInterpResult=true}
    baseJoint {boneName=root}
    planarLimits [
     {drivingBone {boneName=root}, planeTransform {location=<0 0 10>, rotation=<rot 0 0 0>, scale=<1 1 1>}}
     {drivingBone {boneName=spine_03}, planeTransform {location=<0 -20 0>, rotation=<rot 0 15 -89.9999>, scale=<1 1 1>}}
    ]
    rotationLimits [{limitMax=<180 180 180>}, {limitMax=<180 180 180>}]
    rotationOffsets [<0 0 0>, <0 0 0>]
    trailBone {boneName=dyn_jacket_r_02}
    trailRelaxationSpeed {editorCurveData {keys [{arriveTangent=-72.0739, interpMode=RCIM_Cubic, leaveTangent=-72.0739, tangentMode=RCTM_User, time=0, value=35}, <at 1 5 RCIM_Cubic>]}}
   }
   n52 = Trail controller - Bone: dyn_jacket_fr_r_02!(ComponentPose=@n51, Alpha=@n45, FakeVelocity:Vector="0,0,0") -> (Pose) {
    bActorSpaceFakeVel=true
    bShowTrailLocation=true
    debugLifeTime=0.1
    maxDeltaTime=0.0333
    relaxationSpeedScale=3
    alphaScaleBiasClamp {bInterpResult=true}
    baseJoint {boneName=root}
    planarLimits [
     {drivingBone {boneName=root}, planeTransform {location=<0 0 10>, rotation=<rot 0 0 0>, scale=<1 1 1>}}
     {drivingBone {boneName=spine_03}, planeTransform {location=<0 -20 0>, rotation=<rot 0 15 -89.9999>, scale=<1 1 1>}}
    ]
    rotationLimits [{limitMax=<180 180 180>}, {limitMax=<180 180 180>}]
    rotationOffsets [<0 0 0>, <0 0 0>]
    trailBone {boneName=dyn_jacket_fr_r_02}
    trailRelaxationSpeed {editorCurveData {keys [{arriveTangent=-72.0739, interpMode=RCIM_Cubic, leaveTangent=-72.0739, tangentMode=RCTM_User, time=0, value=35}, <at 1 5 RCIM_Cubic>]}}
   }
   n53 = RigidBody!(ComponentPose=Local To Component(LocalPose=Use cached pose 'DeformPost'()), Alpha, OverrideWorldGravity:Vector="0,0,-980", ExternalForce) -> (Pose) {
    bClampLinearTranslationLimitToRefPose=true
    bOverrideWorldGravity=true
    overridePhysicsAsset=&/Pencil_Jacket_MedA_Ubdy_Jacket_Physics
    simulationSpace=BaseBoneSpace
    alphaBoolBlend {blendOption=Linear}
    baseBoneRef {boneName=pelvis}
    componentAppliedLinearAccClamp=<3000 3000 3000>
    componentLinearAccScale=<0.1 0.1 0.3>
    componentLinearVelScale=<0.1 0.1 0.3>
    overrideWorldGravity {z=-980}
    simSpaceSettings {
     maxAngularAcceleration=1000
     maxAngularVelocity=3
     maxLinearAcceleration=700
     maxLinearVelocity=500
     velocityScaleZ=2
     worldAlpha=1
     externalLinearDragV=<0.2 0.2>
    }
   }
   n54 = RigidBody!(ComponentPose=@n53, bAlphaBoolEnabled:bool, OverrideWorldGravity="0,0,-980", ExternalForce) -> (Pose) {
    alphaInputType=Bool
    bClampLinearTranslationLimitToRefPose=true
    bOverrideWorldGravity=true
    overridePhysicsAsset=&/Pencil_Jacket_MedA_Ubdy_Jacket_Physics_NecklaceFE
    simulationSpace=BaseBoneSpace
    alphaBoolBlend {blendOption=Linear}
    baseBoneRef {boneName=root}
    componentLinearAccScale=<0.4 0.4 0.5>
    componentLinearVelScale=<0.4 0.4 0.5>
    overrideWorldGravity {z=-980}
    simSpaceSettings {
     maxAngularAcceleration=1000
     maxAngularVelocity=3
     maxLinearAcceleration=700
     maxLinearVelocity=500
     velocityScaleZ=0
     worldAlpha=1
     externalLinearDragV=<0.2 0.2>
    }
   }
   n55 = "Transform (Modify) Bone - Bone: dyn_shirt_l_MSTR"!(ComponentPose=@n54, Alpha:float=1, Translation="0,0,0", Rotation="0,0,0", Scale="1,1,1") -> (Pose) {
    rotationMode=BMM_Additive
    rotationSpace=BCS_BoneSpace
    scaleMode=BMM_Additive
    scaleSpace=BCS_BoneSpace
    translationMode=BMM_Additive
    translationSpace=BCS_BoneSpace
    alphaBoolBlend {blendOption=Linear}
    boneToModify {boneName=dyn_shirt_l_MSTR}
   }
   n56 = "Transform (Modify) Bone - Bone: dyn_shirt_r_MSTR"!(ComponentPose=@n55, Alpha:float=1, Translation="0,0,0", Rotation="0,0,0", Scale="1,1,1") -> (Pose) {
    rotationMode=BMM_Additive
    rotationSpace=BCS_BoneSpace
    scaleMode=BMM_Additive
    scaleSpace=BCS_BoneSpace
    translationMode=BMM_Additive
    translationSpace=BCS_BoneSpace
    alphaBoolBlend {blendOption=Linear}
    boneToModify {boneName=dyn_shirt_r_MSTR}
   }
   n57 = Blend Poses by bool(
    BlendPose_0=Component To Local(ComponentPose=@n56)
    BlendPose_1=Blend Poses by bool(BlendPose_0=Component To Local(ComponentPose=@n39), BlendPose_1=Component To Local(ComponentPose=@n52), bActiveValue=Get bIsCrouching()) {blendType=Linear}
    bActiveValue=Get bRigidBodyEnabled()
   ) {blendType=Linear}
   n58 = "Transform (Modify) Bone - Bone: dyn_jacket_fr_r_01_MSTR"(
    ComponentPose=Local To Component(LocalPose=Use cached pose 'DeformPost'())
    Alpha=Get ThighRotR() -> (ThighRotR:double)
    Translation="0,2,2"
    Rotation="30,30,0"
    Scale="0.8,1,1"
   ) {
    rotationMode=BMM_Additive
    rotationSpace=BCS_BoneSpace
    scaleMode=BMM_Additive
    scaleSpace=BCS_BoneSpace
    translationMode=BMM_Additive
    translationSpace=BCS_BoneSpace
    alphaBoolBlend {blendOption=Linear}
    boneToModify {boneName=dyn_jacket_fr_r_01_MSTR}
    rotation {pitch=30, yaw=30}
    scale {x=0.8}
    translation {y=2, z=2}
   }
   n59 = "Constraint - dyn_jacket_fr_l_01_MSTR (PPPPPPPP)"!(
    ComponentPose=@n58
    Alpha:float=1
    ConstraintWeights_0=0.2505
    ConstraintWeights_1=0.4852
    ConstraintWeights_2=0.0022
    ConstraintWeights_3=0.001
    ConstraintWeights_4=0.134
    ConstraintWeights_5=0.0615
    ConstraintWeights_6=0.0146
    ConstraintWeights_7=0.0511
   ) -> (Pose) {
    alphaBoolBlend {blendOption=Linear}
    boneToModify {boneName=dyn_jacket_fr_l_01_MSTR}
    constraintSetup [
     {targetBone {boneName=pelvis}}
     {targetBone {boneName=thigh_twist_01_l}}
     {targetBone {boneName=thigh_r}}
     {targetBone {boneName=deform_glute_l}}
     {targetBone {boneName=deform_groin_l}}
     {targetBone {boneName=thigh_twist_01_r}}
     {targetBone {boneName=thigh_l}}
     {targetBone {boneName=deform_groin_r}}
    ]
    constraintWeights [1,1,1,1,1,1,1,1]
   }
   n60 = "Transform (Modify) Bone - Bone: dyn_jacket_fr_l_01_MSTR"!(ComponentPose=@n59, Alpha:float=1, Translation="17.486,7.553,16.716", Rotation="71.201,-93.497,-98.265", Scale="1,1,1") -> (Pose) {
    rotationMode=BMM_Additive
    rotationSpace=BCS_BoneSpace
    scaleMode=BMM_Additive
    scaleSpace=BCS_BoneSpace
    translationMode=BMM_Additive
    translationSpace=BCS_BoneSpace
    alphaBoolBlend {blendOption=Linear}
    boneToModify {boneName=dyn_jacket_fr_l_01_MSTR}
    rotation=<rot 71.201 -93.497 -98.265>
    translation=<17.486 7.553 16.716>
   }
   n61 = RigidBody!(ComponentPose=@n60, Alpha:float, OverrideWorldGravity="0,0,-980", ExternalForce) -> (Pose) {
    bClampLinearTranslationLimitToRefPose=true
    bOverrideWorldGravity=true
    overridePhysicsAsset=&Pencil_Jacket_MedA_Ubdy_Jacket_Physics
    simulationSpace=BaseBoneSpace
    alphaBoolBlend {blendOption=Linear}
    baseBoneRef {boneName=pelvis}
    componentAppliedLinearAccClamp=<3000 3000 3000>
    componentLinearAccScale=<0.4 -4 0.5>
    componentLinearVelScale=<0.4 0.4 0.5>
    overrideWorldGravity {z=-980}
    simSpaceSettings {
     maxAngularAcceleration=1000
     maxAngularVelocity=3
     maxLinearAcceleration=700
     maxLinearVelocity=500
     velocityScaleZ=0
     worldAlpha=1
     externalLinearDragV=<0.2 0.2>
    }
   }
   n62 = Component To Local(ComponentPose=@n61)
   n63 = AND Boolean(A=Get bIsCrouching(), B=Get bRigidBodyEnabled())
   n64 = "Constraint - dyn_badge_MSTR (PPPPPPPP)"!(
    ComponentPose=Local To Component(LocalPose=Blend Poses by bool(BlendPose_0=@n62, BlendPose_1=@n57, bActiveValue=@n63) {blendType=Linear})
    Alpha=1
    ConstraintWeights_0=0.049
    ConstraintWeights_1:float=0.1546
    ConstraintWeights_2:float=0.3462
    ConstraintWeights_3:float=0.1445
    ConstraintWeights_4:float=0.1649
    ConstraintWeights_5:float=0.061
    ConstraintWeights_6:float=0.0612
    ConstraintWeights_7:float=0.0186
   ) -> (Pose) {
    alphaBoolBlend {blendOption=Linear}
    boneToModify {boneName=dyn_badge_MSTR}
    constraintSetup [
     {targetBone {boneName=spine_02}}
     {targetBone {boneName=spine_03}}
     {targetBone {boneName=spine_04}}
     {targetBone {boneName=spine_05}}
     {targetBone {boneName=clavicle_l}}
     {targetBone {boneName=upperarm_twist_01_l}}
     {targetBone {boneName=deform_upperarm_fr_l}}
     {targetBone {boneName=upperarm_twist_02_l}}
    ]
    constraintWeights [1,1,1,1,1,1,1,1]
   }
   n65 = "Transform (Modify) Bone - Bone: dyn_badge"(ComponentPose=@n64, Alpha=Get LArmUp() -> (LArmUp:double), Translation="0.715,0.195,-0.076", Rotation="-18.379,6.715,-3.209", Scale="1,1,1") {
    rotationMode=BMM_Additive
    rotationSpace=BCS_BoneSpace
    scaleMode=BMM_Additive
    scaleSpace=BCS_BoneSpace
    translationMode=BMM_Additive
    translationSpace=BCS_BoneSpace
    alphaBoolBlend {blendOption=Linear}
    boneToModify {boneName=dyn_badge}
    rotation=<rot -18.379 6.715 -3.209>
    translation=<0.715 0.195 -0.076>
   }
   n66 = "Transform (Modify) Bone - Bone: dyn_badge"!(ComponentPose=@n65, bAlphaBoolEnabled:bool=True, Translation="-0.450,0.941,-0.145", Rotation="18,0,0", Scale="1,1,1") -> (Pose) {
    alphaInputType=Bool
    rotationMode=BMM_Additive
    rotationSpace=BCS_BoneSpace
    scaleMode=BMM_Additive
    scaleSpace=BCS_BoneSpace
    translationMode=BMM_Additive
    translationSpace=BCS_BoneSpace
    alphaBoolBlend {blendInTime=0.1, blendOption=Linear, blendOutTime=0.1}
    boneToModify {boneName=dyn_badge}
    rotation {pitch=18}
    translation=<-0.45 0.941 -0.145>
   }
   n67 = "Transform (Modify) Bone - Bone: dyn_badge"(ComponentPose=@n66, bAlphaBoolEnabled=True, Translation="0.6,0,0", Rotation="-12,0,0", Scale="1,1,1") {
    alphaInputType=Bool
    rotationMode=BMM_Additive
    rotationSpace=BCS_BoneSpace
    scaleMode=BMM_Additive
    scaleSpace=BCS_BoneSpace
    translationMode=BMM_Additive
    translationSpace=BCS_BoneSpace
    alphaBoolBlend {blendInTime=0.1, blendOption=Linear, blendOutTime=0.1}
    boneToModify {boneName=dyn_badge}
    rotation {pitch=-12}
    translation {x=0.6}
   }
   n68 = "Transform (Modify) Bone - Bone: collar_fr_mstr"(
    ComponentPose=@n67
    Alpha=Get L upperarm alpha() -> (L upperarm alpha:double)
    Translation="0,0,0"
    Rotation="-44.099,-13.947,35.524"
    Scale="0.8,1,1"
   ) {
    rotationMode=BMM_Additive
    rotationSpace=BCS_BoneSpace
    scaleMode=BMM_Additive
    scaleSpace=BCS_BoneSpace
    translationMode=BMM_Additive
    translationSpace=BCS_BoneSpace
    alphaBoolBlend {blendOption=Linear}
    alphaScaleBiasClamp {bInterpResult=true, interpSpeedDecreasing=20, interpSpeedIncreasing=20}
    boneToModify {boneName=collar_fr_mstr}
    rotation=<rot -44.0989 -13.9471 35.5241>
    scale {x=0.8}
   }
   n69 = "Constraint - collar_br_mstr (P)"!(ComponentPose=@n68, bAlphaBoolEnabled:bool=Get Moving Crouch(), ConstraintWeights_0=1) -> (Pose) {
    alphaInputType=Bool
    alphaBoolBlend {blendInTime=0.1, blendOption=Linear, blendOutTime=0.1}
    boneToModify {boneName=collar_br_mstr}
    constraintSetup [{targetBone {boneName=head}}]
    constraintWeights [1]
   }
   Save cached pose 'PreCrouch'(Pose=Component To Local(ComponentPose=@n69)) {cacheName=PreCrouch}
  ]}
 EventGraph {nodes [
   Event BlueprintUpdateAnimation()
   > n1 = Sequence!(execute) -> (then_0, then_1, then_2:Exec, then_3:Exec, then_4:Exec, then_5:Exec, then_6:Exec)
   n2 = GetOwnerSkeletalMeshComponent()
   n3 = BreakRotator(InRot=BreakTransform(InTransform=GetDeltaTransformFromRefPose(self=@n2, BoneName=upperarm_l, BaseName=clavicle_l)).Rotation)
   Set L upperarm alpha(
    execute=[@n1.then_2, @n12.then]
    L upperarm alpha:double="Max (Float)"(A=MapRangeClamped(Value=@n3.Yaw, InRangeA=60, InRangeB=80), B=MapRangeClamped(Value=@n3.Yaw, InRangeA=30, InRangeB=80))
   ) -> (Output_Get:double)
   n4 = Get Skel_Mesh()
   > Set thigh_rot_otr_l(thigh_rot_otr_l:double=GetAlphaFromReferenceTransformDelta(SkeletalMeshComponent=@n4, BoneName=thigh_l, BaseName=pelvis, DeltaComponentType=EulerY, InputRange=(X=20,Y=50), TwistAxis="1,0,0")) -> (Output_Get:double)
   > Set thigh_rot_otr_r(thigh_rot_otr_r:double=GetAlphaFromReferenceTransformDelta(SkeletalMeshComponent=@n4, BoneName=thigh_r, BaseName=pelvis, DeltaComponentType=EulerY, InputRange=(X=20,Y=50), TwistAxis="1,0,0")) -> (Output_Get:double)
   n5 = Get Skel_Mesh()
   > Set thigh_rot_inr_l(thigh_rot_inr_l:double=GetAlphaFromReferenceTransformDelta(SkeletalMeshComponent=@n5, BoneName=thigh_l, BaseName=pelvis, DeltaComponentType=EulerY, InputRange=(X=10,Y=-6), TwistAxis="1,0,0")) -> (Output_Get:double)
   > Set thigh_rot_inr_r(thigh_rot_inr_r:double=GetAlphaFromReferenceTransformDelta(SkeletalMeshComponent=@n5, BoneName=thigh_r, BaseName=pelvis, DeltaComponentType=EulerY, InputRange=(X=0,Y=-10), TwistAxis="1,0,0")) -> (Output_Get:double)
   > Set ThighRotR(ThighRotR:double=GetAlphaFromReferenceTransformDelta(SkeletalMeshComponent=Get Skel_Mesh(), BoneName=thigh_r, BaseName=pelvis, InputRange=(X=-20,Y=-100), TwistAxis="1,0,0")) -> (Output_Get:double)
   n6 = Branch(execute=@n1.then_4, Condition=AND Boolean!(A=Get bIsCrouching(), B=Get bIsAccelerating2D(), C:bool) -> (ReturnValue))
   > Set Moving Crouch(Moving Crouch:bool=true) -> (Output_Get:bool)
   Set Moving Crouch(execute=@n6.else, Moving Crouch=false)
   n7 = Branch(execute=@n1.then_5, Condition=Get bIsCrouching())
   > Set newCrouching(newCrouching:bool=true) -> (Output_Get:bool)
   Set newCrouching(execute=@n7.else, newCrouching=false)
   n8 = Branch(execute=@n1.then_6, Condition=OR Boolean(A=Get bIsJumping(), B=Get bIsFalling()))
   > Set JumpingFalling(JumpingFalling:bool=true) -> (Output_Get:bool)
   Set JumpingFalling(execute=@n8.else, JumpingFalling=false)
   n9 = Gravity Override State Function DEPRECATED(
    execute=@n1.then_0
    CharacterPartAnimInstance=Self-Reference()
    GameTimeInSeconds=GetGameTimeInSeconds()
    GravityOverrideParams=Get Gravity Override Params()
   )
   > Set GravityOverride(GravityOverride=@n9)
   n10 = BreakTransform(InTransform=GetDeltaTransformFromRefPose(self=GetOwnerSkeletalMeshComponent(), BoneName=spine_05, BaseName=clavicle_l))
   Set LArmUp(execute=@n1.then_1, LArmUp:double=MapRangeClamped(Value=BreakRotator(InRot=@n10.Rotation).Yaw, InRangeA=20, InRangeB=-10)) -> (Output_Get:double)
   n11 = BreakTransform(InTransform=GetDeltaTransformFromRefPose(self=GetOwnerSkeletalMeshComponent(), BoneName=spine_05, BaseName=clavicle_l))
   > n12 = Set RArmUp(RArmUp:double=MapRangeClamped(Value=BreakRotator(InRot=@n11.Rotation).Yaw, InRangeA=3, InRangeB=-11)) -> (Output_Get:double)
   n13 = PlayerMovement(execute=@n1.then_3, AnimInstance=Self-Reference(), XDotProductCoefficient=1, YDotProductCoefficient=1, ZDotProductCoefficient=1)
   n14 = MapRangeClamped(Value=@n13.YMovement, InRangeA=0, InRangeB=-1)
   n15 = float * float(A="Absolute (Float)"(A=@n13), B=0.5)
   > Set Bwd Movement(Bwd Movement:double=MapRangeClamped(
     Value=float + float(A=@n15, B=@n14) {is=title K2Node_CommutativeAssociativeBinaryOperator}
     InRangeA=0
     InRangeB=1.1
     OutRangeA=0.1
     OutRangeB=0
    )) -> (Output_Get:double)
   n16 = MapRangeClamped(Value=@n13, InRangeA=-0.2, InRangeB=1)
   n17 = float * float(A="Absolute (Float)"(A=@n13.YMovement), B=0.2)
   > Set Right Movement(Right Movement:double=MapRangeClamped(Value=float + float(A=@n17, B=@n16), InRangeA=0, InRangeB=1.2, OutRangeB=0)) -> (Output_Get:double)
   n18 = MapRangeClamped(Value=@n13, InRangeA=0.2, InRangeB=-1, OutRangeB=0.3)
   n19 = float * float(A="Absolute (Float)"(A=@n13.YMovement), B=0.2)
   > Set Left Movement(Left Movement:double=MapRangeClamped(Value=float + float(A=@n19, B=@n18), InRangeA=0, InRangeB=1.2, OutRangeB=0.5)) -> (Output_Get:double)
   Event OnCharacterCustomizationLoaded()
   > Set PencilApple(PencilApple:bool=CharacterItemDefinitionHasAnyMetaTag(Tags=Make Array("[0]"=(TagName="Cosmetics.MetaTags.Pencil.Head.Apple")))) -> (Output_Get:bool)
   > Set PencilGrape(PencilGrape:bool=CharacterItemDefinitionHasAnyMetaTag(Tags=Make Array!(
      "[0]"=(TagName="Cosmetics.MetaTags.Pencil.Head.Grape")
      "[1]":Gameplay Tag Structure=(TagName="Cosmetics.MetaTags.Pencil.Head")
      "[2]":Gameplay Tag Structure=(TagName="Cosmetics.MetaTags.Pencil.Head.Kiwi")
      "[3]":Gameplay Tag Structure=(TagName="Cosmetics.MetaTags.Pencil.Head.Lime")
      "[4]":Gameplay Tag Structure
     ) -> (Array))) -> (Output_Get:bool)
   > Set PencilRaspberry(PencilRaspberry:bool=CharacterItemDefinitionHasAnyMetaTag(Tags=Make Array!("[0]"=(TagName="Cosmetics.MetaTags.Pencil.Head.Raspberry")) -> (Array))) -> (Output_Get:bool)
   Event BlueprintInitializeAnimation()
   > Set Skel_Mesh(Skel_Mesh=GetOwnerSkeletalMeshComponent())
   MapRangeClamped(Value=0, InRangeA=0, InRangeB=1.2, OutRangeB=0.5)
  ]}
]
variables [
 Bwd Movement {type=number}
 Gravity Override Params {type=GravityOverrideParamsStruct}
 GravityOverride {type=Vector}
 isAimingPistol {type=boolean}
 isAimingRifle {type=boolean}
 isMelee {type=boolean}
 JumpingFalling {type=boolean}
 L upperarm alpha {type=number}
 LArmUp {type=number}
 Left Movement {type=number}
 Moving Crouch {type=boolean}
 newCrouching {type=boolean}
 PencilApple {type=boolean}
 PencilGrape {type=boolean}
 PencilRaspberry {type=boolean}
 RArmUp {type=number}
 Right Movement {type=number}
 Skel_Mesh {type=SkeletalMeshComponent}
 thigh_rot_inr_l {type=number}
 thigh_rot_inr_r {type=number}
 thigh_rot_otr_l {type=number}
 thigh_rot_otr_r {type=number}
 ThighRotR {type=number}
]
